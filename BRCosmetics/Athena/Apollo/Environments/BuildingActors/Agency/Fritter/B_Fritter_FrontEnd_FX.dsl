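; Front-end FX actor for the Fritter device, showing its energy build-up and destroyed aftermath in step with calendar events.
; On BeginPlay it watches two calendar events, mapping the stages event's elapsed ratio to stages 1–4 and the aftermath event to stage 5.
; SetStages switches on that number to tune two particle systems and a point light, hiding the in-device FX at stage 0.
; Stage 5 async-loads aftermath particles and a cracked-glass material, and hides the light.
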(event EventBeginPlay
  (Utilities|FlowControl|Sequence
    (:then_0
      (MCPContext|WatchCalendarEvent (Variables|Default|GetAftermathCalendarEvent)
        (:OnEventActive
          (Variables|Default|SetStageNum 5))))
    (:then_1
      (bind (_eventname _timeuntilend _timesincebegin _timespanratio) (MCPContext|WatchCalendarEvent (Variables|Default|GetStagesActiveCalendarEvent))
        (:OnEventActive
          (if (>= _timespanratio 0.23)
            (if (>= _timespanratio 0.56)
              (if (>= _timespanratio 0.78)
                (Variables|Default|SetStageNum 4)
                (else
                  (Variables|Default|SetStageNum 3)))
              (else
                (Variables|Default|SetStageNum 2)))
            (else
              (Variables|Default|SetStageNum 1)))))))
  (CallFunction|SetStages))

(event Custom|SetStages
  (bind _p_fritter_energywaves_onprong (Variables|Default|GetP_Fritter_EnergyWaves_OnProng))
  (bind _p_fritter_energywaves_indevice (Variables|Default|GetP_Fritter_EnergyWaves_InDevice))
  (bind _p_fritter_energywaves_indevice_2 (Variables|Default|GetP_Fritter_EnergyWaves_InDevice))
  (Effects|Components|ParticleSystem|SetFloatParameter _p_fritter_energywaves_indevice_2 "ChargeEffectTurnOn" 0.0) ; Turn off activate fx
  (Effects|Components|ParticleSystem|SetFloatParameter _p_fritter_energywaves_indevice_2 "ParticleLight" 0.0) ; Turn off particle light
  (switch int (Variables|Default|GetStageNum)
    (:0
      ;[1 1 1] Stage 0
      (Rendering|SetVisibility _p_fritter_energywaves_indevice false false))
    (:1
      ;[0.6218 1 0.5705] Stage 1
      (Rendering|SetVisibility _p_fritter_energywaves_indevice true false)
      (Effects|Components|ParticleSystem|SetFloatParameter _p_fritter_energywaves_indevice "BG_Energy_Add" 1.0)
      (Effects|Components|ParticleSystem|SetFloatParameter _p_fritter_energywaves_indevice "BG_Energy_Add_Opacity" 0.05)
      (Effects|Components|ParticleSystem|SetFloatParameter _p_fritter_energywaves_indevice "Chromatic" 0.0)
      (Effects|Components|ParticleSystem|SetFloatParameter _p_fritter_energywaves_indevice "WaveDistortion" 0.0)
      (Effects|Components|ParticleSystem|SetFloatParameter _p_fritter_energywaves_indevice "BG_Energy" 0.0)
      (Effects|Components|ParticleSystem|SetFloatParameter _p_fritter_energywaves_indevice "ProngEnergyConnection" 0.0)
      (Effects|Components|ParticleSystem|SetFloatParameter _p_fritter_energywaves_indevice "CoreElements" 0.0) ; Energy Arc between prongs and glass surface
      (Effects|Components|ParticleSystem|SetVectorParameter _p_fritter_energywaves_indevice "FlareColor" "15.000000,6.000000,2.500000")
      (Effects|Components|ParticleSystem|SetVectorParameter _p_fritter_energywaves_indevice "FlareSize" "0.200000,1.000000,1.000000")
      (Effects|Components|ParticleSystem|SetFloatParameter _p_fritter_energywaves_indevice "SpawnRate" 0.0)
      (Effects|Components|ParticleSystem|SetFloatParameter _p_fritter_energywaves_onprong "SpawnRateOnProng" 0.5) ; Energy Arc between prongs and glass surface
      (Rendering|Components|Light|SetLightColor (Variables|Default|GetPointLight) "(R=0.250000,G=0.150234,B=0.088941,A=1.000000)" true))
    (:2
      ;[1 0.812 0.5021] Stage 2
      (Rendering|SetVisibility _p_fritter_energywaves_indevice true false)
      (Effects|Components|ParticleSystem|SetFloatParameter _p_fritter_energywaves_indevice "BG_Energy_Add" 1.0)
      (Effects|Components|ParticleSystem|SetFloatParameter _p_fritter_energywaves_indevice "BG_Energy_Add_Opacity" 0.5)
      (Effects|Components|ParticleSystem|SetFloatParameter _p_fritter_energywaves_indevice "Chromatic" 0.0)
      (Effects|Components|ParticleSystem|SetFloatParameter _p_fritter_energywaves_indevice "WaveDistortion" 0.1)
      (Effects|Components|ParticleSystem|SetFloatParameter _p_fritter_energywaves_indevice "BG_Energy" 0.0)
      (Effects|Components|ParticleSystem|SetFloatParameter _p_fritter_energywaves_indevice "ProngEnergyConnection" 1.0)
      (Effects|Components|ParticleSystem|SetVectorParameter _p_fritter_energywaves_indevice "FlareSize" "0.600000,1.000000,1.000000")
      (Effects|Components|ParticleSystem|SetVectorParameter _p_fritter_energywaves_indevice "FlareColor" "10.000000,1.000000,0.100000")
      (Effects|Components|ParticleSystem|SetVectorParameter _p_fritter_energywaves_indevice "CoreMeshSize" "0.400000,0.400000,0.400000")
      (Effects|Components|ParticleSystem|SetFloatParameter _p_fritter_energywaves_indevice "SpawnRate" 0.2) ; Energy Arc between prongs and glass surface
      (Effects|Components|ParticleSystem|SetFloatParameter _p_fritter_energywaves_onprong "WaveDistortion" 0.3)
      (Effects|Components|ParticleSystem|SetFloatParameter _p_fritter_energywaves_onprong "SpawnRateOnProng" 1.0) ; Energy Arc between prongs and glass surface
    )
    (:3
      ;[1 0.3699 0.1586] Stage 3
      (Rendering|SetVisibility _p_fritter_energywaves_indevice true false)
      (Effects|Components|ParticleSystem|SetFloatParameter _p_fritter_energywaves_indevice "BG_Energy_Add" 0.0)
      (Effects|Components|ParticleSystem|SetFloatParameter _p_fritter_energywaves_indevice "WaveDistortion" 0.8)
      (Effects|Components|ParticleSystem|SetFloatParameter _p_fritter_energywaves_indevice "BG_Energy" 1.0)
      (Effects|Components|ParticleSystem|SetFloatParameter _p_fritter_energywaves_indevice "ProngEnergyConnection" 1.0)
      (Effects|Components|ParticleSystem|SetFloatParameter _p_fritter_energywaves_indevice "SpawnRate" 1.0) ; Energy Arc between prongs and glass surface
      (Effects|Components|ParticleSystem|SetVectorParameter _p_fritter_energywaves_indevice "FlareSize" "0.800000,1.000000,1.000000")
      (Effects|Components|ParticleSystem|SetVectorParameter _p_fritter_energywaves_indevice "CoreMeshSize" "1.000000,1.000000,1.000000")
      (Effects|Components|ParticleSystem|SetFloatParameter _p_fritter_energywaves_onprong "WaveDistortion" 0.8)
      (Effects|Components|ParticleSystem|SetFloatParameter _p_fritter_energywaves_onprong "SpawnRateOnProng" 1.0) ; Energy Arc between prongs and glass surface
      (Effects|Components|ParticleSystem|SetFloatParameter _p_fritter_energywaves_onprong "SparkShower" 1.0) ; Energy Arc between prongs and glass surface
    )
    (:4
      ;[1 0.08896 0.02924] Stage Max
      (Rendering|SetVisibility _p_fritter_energywaves_indevice true false)
      (Effects|Components|ParticleSystem|SetVectorParameter _p_fritter_energywaves_indevice "OverallBrightness" "3.500000,3.000000,3.000000")
      (Effects|Components|ParticleSystem|SetFloatParameter _p_fritter_energywaves_indevice "BG_Energy_Add" 0.0)
      (Effects|Components|ParticleSystem|SetFloatParameter _p_fritter_energywaves_indevice "WaveDistortion" 1.1)
      (Effects|Components|ParticleSystem|SetFloatParameter _p_fritter_energywaves_indevice "BG_Energy" 1.0)
      (Effects|Components|ParticleSystem|SetFloatParameter _p_fritter_energywaves_indevice "ProngEnergyConnection" 1.0)
      (Effects|Components|ParticleSystem|SetFloatParameter _p_fritter_energywaves_indevice "SpawnRate" 0.0) ; Energy Arc between prongs and glass surface
      (Effects|Components|ParticleSystem|SetFloatParameter _p_fritter_energywaves_indevice "Wave_Small_CamFacing" 0.0) ; Energy Arc between prongs and glass surface
      (Effects|Components|ParticleSystem|SetFloatParameter _p_fritter_energywaves_indevice "Max_Element" 1.0) ; Energy Arc between prongs and glass surface
      (Effects|Components|ParticleSystem|SetVectorParameter _p_fritter_energywaves_indevice "FlareSize" "1.500000,1.000000,1.000000")
      (Effects|Components|ParticleSystem|SetVectorParameter _p_fritter_energywaves_indevice "CoreMeshSize" "1.300000,1.300000,1.300000")
      (Effects|Components|ParticleSystem|SetVectorParameter _p_fritter_energywaves_indevice "FlareColor" "4.000000,1.200000,0.400000")
      (Effects|Components|ParticleSystem|SetFloatParameter _p_fritter_energywaves_onprong "WaveDistortion" 1.0)
      (Effects|Components|ParticleSystem|SetFloatParameter _p_fritter_energywaves_onprong "SpawnRateOnProng" 0.0) ; Energy Arc between prongs and glass surface
      (Effects|Components|ParticleSystem|SetFloatParameter _p_fritter_energywaves_onprong "SparkShower" 1.5) ; Energy Arc between prongs and glass surface
      (Effects|Components|ParticleSystem|SetVectorParameter _p_fritter_energywaves_onprong "ProngFlareColor" "15.000000,6.000000,1.000000")
      (Rendering|Components|Light|SetLightColor (Variables|Default|GetPointLight) "(R=5.000000,G=0.300000,B=0.024747,A=1.000000)" true))
    (:5
      ;[1 1 1] Destroyed Stage
      (bind _object (Utilities|AsyncLoadAsset (Variables|Default|GetP_Fritter_EnergyWavesinDeviceAftermath))
        (:then
          (bind _object_2 (Utilities|AsyncLoadAsset (Variables|Default|GetCrackGlassMaterial))
            (:Completed
              (bind _asmaterial (Utilities|Casting|CastToMaterial _object_2))
              (Rendering|Material|SetMaterial (Variables|Default|GetMachine) 0 _asmaterial)
              ;;[1 1 1] Glass cracks appears on post event
              (Rendering|Material|SetScalarParameterValueonMaterials (Variables|Default|GetMachine) "GlassCrack" 1.0)
              (Rendering|SetVisibility (Variables|Default|GetPointLight) false false))))
        (:Completed
          (bind _asparticle_system (Utilities|Casting|CastToParticleSystem _object))
          (Effects|Components|ParticleSystem|SetTemplate (Variables|Default|GetP_Fritter_EnergyWaves_InDevice_Aftermath) _asparticle_system))))))

(fn ConstructionScript)
