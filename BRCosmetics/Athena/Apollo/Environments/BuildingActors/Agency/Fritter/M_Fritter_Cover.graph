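# Additive unlit material for mesh particles, drawing a glowing shell effect, apparently for the Fritter cover.
# View-derived UVs scroll a wave texture, which is added to a sky cubemap reflection tinted by Fresnel-blended shell colors.
# Particle color, invertible Fresnel opacity, depth-fade edge brightening and an optional V-gradient fade shape the emissive.
# Shading path switches give Forward and Mobile a cheaper version without edge fades, gradient or two-color tint.
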Material /BRCosmetics/Athena/Apollo/Environments/BuildingActors/Agency/Fritter/M_Fritter_Cover
graph {
 properties {
  bUsedWithMeshParticles=true
  bUsedWithNiagaraMeshParticles=true
  blendMode=BLEND_Additive
  pixelDepthOffsetMode=PDOM_Legacy
  shadingModel=MSM_Unlit
  usedShadingModels=Unlit
 }
 nodes [
  n1 = ReflectionVectorWS()
  n2 = CameraVectorWS()
  n3 = ComponentMask(None=Transform(None=CrossProduct(A=@n1, B=@n2))) {g=true, r=true}
  n4 = ScalarParameter() {defaultValue=2, parameterName=NoiseTiling}
  n5 = TextureSample(UVs=Add(A=Multiply(A=@n3, B=@n4) {constB=2}, B=Time())) {texture=/Game/Global/Textures/Default/Tileable_Patterns/T_WavePattern_01}
  n6 = Constant() {r=0.3}
  n7 = ScalarParameter() {defaultValue=1, parameterName=EdgeFade}
  n8 = Saturate(None=Multiply(A=MaterialFunctionCall("FadeSensitivity (optional)"=@n7) {materialFunction=&FadeBasedOnViewAngle}.Result) {constB=4})
  n9 = Constant() {r=1}
  n10 = Multiply(
   A=Multiply(A=Desaturation(None=@n5.RGB, Fraction=@n6), B=ScalarParameter() {defaultValue=1, parameterName=NoiseBrightness})
   B=ShadingPathSwitch(Default=@n8, Deferred=@n8, Forward=@n9, Mobile=@n9)
  )
  n11 = ParticleColor()
  n12 = TextureSampleParameterCube!(UVs=ReflectionVectorWS(), Apply View MipBias) {
   parameterName=Param
   samplerType=SAMPLERTYPE_LinearColor
   texture=/BRCosmetics/Packages/Fortress_Sky/SkyDome/MaterialInstances/SkyDome_S8/S8_Day_HDR
  }
  n13 = VectorParameter() {parameterName=ShellColor_Low, defaultValue=<rgba 0.3663 6 5.7014 1>}
  n14 = Fresnel() {baseReflectFraction=0.2, exponent=2}
  n15 = Multiply(
   A=LinearInterpolate(A=VectorParameter() {parameterName=ShellColor_High, defaultValue=<rgba 6 1.5669 0.667 1>}.RGB, B=@n13.RGB, Alpha=@n14)
   B=@n14
  )
  n16 = ShadingPathSwitch(Default=@n15, Deferred=@n15, Forward=@n13.RGB, Mobile=@n13.RGB)
  n17 = Fresnel() {baseReflectFraction=0.2, exponent=1}
  n18 = LinearInterpolate(A=@n17, B=OneMinus(None=@n17), Alpha=ScalarParameter() {parameterName=Invert Opacity Fresnel})
  n19 = Multiply(
   A=Add(A=Multiply(A=@n10, B=@n11.RGB), B=Multiply(A=@n12.RGB, B=@n16))
   B=Multiply(A=Multiply(A=@n11.A, B=@n18), B=ScalarParameter() {defaultValue=0.3, parameterName=Opacity})
  )
  n20 = LinearInterpolate(
   A=Multiply(A=@n19, B=ScalarParameter() {defaultValue=50, parameterName=Depth Fade Edge Brightness}) {constB=50}
   B=@n19
   Alpha=DepthFade(FadeDistance=ScalarParameter() {parameterName=DepthFade})
  )
  n21 = MaterialFunctionCall() {materialFunction=&LinearGradient}
  n22 = LinearInterpolate(A=@n20, B=Multiply(A=@n20, B=OneMinus(None=@n21.VGradient)), Alpha=ScalarParameter() {parameterName=LinearGradientFadeV})
  MaterialFunctionCall(Emissive=ShadingPathSwitch(Default=@n22, Deferred=@n22, Forward=@n19, Mobile=@n19)) {materialFunction=&MF_FX_EmissiveAdjustments}
 ]
}
