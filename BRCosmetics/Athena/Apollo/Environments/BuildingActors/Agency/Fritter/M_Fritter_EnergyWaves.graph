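# Translucent unlit mesh-particle material for the Fritter Agency energy waves, drawing glowing red and cyan bands.
# Two sine bands from noise-distorted vertical UV gradients scroll by time or MPC Loudness, tuned by particle dynamic parameters.
# Opacity combines a gradient mask, particle alpha, camera, edge and depth fades, plus optional flicker.
# Vertex-color-masked noise drives WPO; forward and mobile paths drop WPO, detail distortion and those three fades.
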Material /BRCosmetics/Athena/Apollo/Environments/BuildingActors/Agency/Fritter/M_Fritter_EnergyWaves
graph {
 properties {
  bEnableResponsiveAA=true
  bUsedWithMeshParticles=true
  bUsedWithNiagaraMeshParticles=true
  blendMode=BLEND_Translucent
  pixelDepthOffsetMode=PDOM_Legacy
  shadingModel=MSM_Unlit
  usedShadingModels=Unlit
 }
 nodes [
  Constant() {r=0.02}
  n1 = ParticleColor()
  n2 = TextureCoordinate()
  n3 = SphereMask(
   A=Power(Base=MaterialFunctionCall(Float2=@n2) {materialFunction=&BreakOutFloat2Components}.G) {constExponent=0.6}
   B=Constant() {r=0.5}
  ) {attenuationRadius=0.5, hardnessPercent=0}
  n4 = Multiply(A=@n3, B=@n3)
  n5 = TextureCoordinate()
  n6 = ScalarParameter() {defaultValue=1, parameterName=WaveTileGradientStretch}
  n7 = Multiply(
   A=DynamicParameter() {paramNames [WaveDistortion, WPO, Tile_Wave, TimeOffset]}.WaveDistortion
   B=ScalarParameter() {defaultValue=1, parameterName=WaveDistortion}
  )
  n8 = VectorParameter() {parameterName=UV_Distortion_Tile, defaultValue=<rgba 2 0.1 20 1>}
  n9 = Multiply(A=TextureCoordinate(), B=AppendVector(A=@n8.R, B=@n8.R))
  n10 = AppendVector(A=Constant(), B=LinearInterpolate(Alpha=ScalarParameter() {defaultValue=1, parameterName=WPO_PanSpeed}) {constB=-0.1})
  n11 = TextureSample(UVs=Panner(Coordinate=@n9, Speed=@n10) {speedY=-0.1}) {samplerType=SAMPLERTYPE_LinearColor, texture=&T_NoisePack_LinearColor}
  n12 = Saturate(None=Multiply(A=MaterialFunctionCall(Float2=@n9) {materialFunction=&BreakOutFloat2Components}.G) {constB=10})
  n13 = Subtract(
   A=Power(
    Base=MaterialFunctionCall(Float2=@n5) {materialFunction=&BreakOutFloat2Components}.G
    Exp=@n6
   )
   B=Multiply(A=LinearInterpolate(B=@n7, Alpha=@n11.R) {constB=0.85}, B=@n12)
  )
  n14 = Multiply(
   A=ScalarParameter() {defaultValue=30, parameterName=Tile_Wave}
   B=DynamicParameter() {paramNames [WaveDistortion, WPO, Tile_Wave, TimeOffset]}.Tile_Wave
  )
  n15 = LinearInterpolate(
   A=@n14
   B=Divide(A=@n14) {constB=1.2}
   Alpha=MaterialFunctionCall(Distance=ScalarParameter() {defaultValue=20000, parameterName=NearCameraFadeDis}) {materialFunction=&Near_Camera_Fade}.Result
  ) {constA=1, constB=2}
  n16 = ScalarParameter() {defaultValue=-6, parameterName=WavePanningSpeed}
  n17 = Multiply(A=Add(A=Time()) {constB=-0.01}, B=@n16) {constB=-6}
  n18 = CollectionParameter() {collection=/Game/Athena/Apollo/Environments/BuildingActors/Agency/Fritter/MPC_Fritter_EnergyWave, parameterName=Loudness}
  n19 = ScalarParameter() {parameterName=UseMPC_Speed}
  n20 = Sine(None=Add(A=Multiply(A=@n13, B=@n15), B=LinearInterpolate(A=@n17, B=@n18, Alpha=@n19)))
  n21 = Saturate(None=ConstantBiasScale(None=Add(A=@n20) {constB=0.4}))
  n22 = Multiply(A=@n21) {constB=0.3}
  n23 = TextureCoordinate()
  n24 = TextureSample(UVs=Panner(Coordinate=@n9, Speed=@n10) {speedY=-0.1}) {samplerType=SAMPLERTYPE_LinearColor, texture=&T_NoisePack_LinearColor}
  n25 = Multiply(A=LinearInterpolate(B=@n7, Alpha=@n24.R) {constB=0.85}, B=@n12)
  n26 = Multiply(A=@n25) {constB=1.1}
  n27 = OneMinus(None=MaterialFunctionCall() {materialFunction=&LinearGradient}.VGradient)
  n28 = Multiply(A=@n27, B=@n27)
  n29 = Multiply(A=Multiply(A=@n28, B=@n28), B=ScalarParameter() {defaultValue=1, parameterName=WaveDistortion_DetailExp})
  n30 = Panner(Coordinate=Multiply(A=@n9, B=AppendVector(A=@n8.B, B=@n8.B))) {speedY=-0.1}
  n31 = ScalarParameter() {defaultValue=4, parameterName=FlickerInitialSpeed}
  n32 = Constant()
  n33 = Multiply(
   A=LinearInterpolate(B=ScalarParameter() {defaultValue=4, parameterName=WaveDistortion_Detail}, Alpha=Saturate(None=@n29))
   B=Multiply(
    A=TextureSample(UVs=@n30) {samplerType=SAMPLERTYPE_LinearColor, texture=&T_NoisePack_LinearColor}.R
    B=MaterialFunctionCall!(FlickerInitialSpeed=@n31, FlickerRandom=@n32) {materialFunction=/Game/Effects/Fort_Effects/Materials/_Functions/MF_Flicker}.Result
   )
  )
  n34 = Constant()
  n35 = Subtract(
   A=Power(
    Base=MaterialFunctionCall(Float2=@n23) {materialFunction=&BreakOutFloat2Components}.G
    Exp=@n6
   )
   B=Add(A=LinearInterpolate(A=@n25, B=@n26, Alpha=@n12), B=ShadingPathSwitch(Default=@n33, Deferred=@n33, Forward=@n34, Mobile=@n34)) {constB=0}
  )
  n36 = Multiply(A=Time(), B=@n16) {constB=-6}
  n37 = CollectionParameter() {collection=&MPC_Fritter_EnergyWave, parameterName=Loudness}
  n38 = ScalarParameter() {parameterName=UseMPC_Speed}
  n39 = Sine(None=Add(A=Multiply(A=@n35, B=@n15), B=LinearInterpolate(A=@n36, B=@n37, Alpha=@n38)))
  n40 = Power(Base=Saturate(None=Add(A=@n39) {constB=-0.3}))
  n41 = Multiply(A=@n40) {constB=1.5}
  n42 = Multiply(A=@n1.A, B=Multiply(A=@n4, B=Add(A=@n22, B=@n41) {desc=Combine A + B}))
  n43 = LinearInterpolate(
   A=ScalarParameter() {defaultValue=1, parameterName=NearCameraFade_NearOpacity}
   B=ScalarParameter() {defaultValue=2, parameterName=NearCameraFade_FarOpacity}
   Alpha=MaterialFunctionCall(Distance=ScalarParameter() {defaultValue=20000, parameterName=NearCameraFadeDis}) {materialFunction=&Near_Camera_Fade}.Result
  ) {constA=1, constB=2}
  n44 = Multiply(
   A=MaterialFunctionCall("FadeSensitivity (optional)"=ScalarParameter() {defaultValue=3, parameterName=EdgeFadeExp}) {materialFunction=&FadeBasedOnViewAngle}.Result
   B=ScalarParameter() {defaultValue=3, parameterName=EdgeFadeHardness}
  ) {constB=3}
  n45 = DepthFade(Opacity=Multiply(A=Multiply(A=@n42, B=@n43), B=Saturate(None=@n44)), FadeDistance=ScalarParameter() {parameterName=DepthFade})
  n46 = ShadingPathSwitch(Default=@n45, Deferred=@n45, Forward=@n42, Mobile=@n42)
  n47 = CollectionParameter() {collection=&MPC_Fritter_EnergyWave, parameterName=RandomOpacityFlickSpeed}
  n48 = DynamicParameter() {paramNames [WaveDistortion, WPO, Tile_Wave, TimeOffset]}
  LinearInterpolate(
   A=@n46
   B=Multiply(
    A=MaterialFunctionCall(FlickerInitialSpeed=@n47, FlickerRandom=@n48.TimeOffset) {materialFunction=&MF_Flicker}.Result
    B=@n46
   )
   Alpha=ScalarParameter() {parameterName=UseRandomOpacity}
  ) {constA=1}
  n49 = Constant3Vector() {constant=<rgba 1 0.0487 0 0>}
  n50 = Constant3Vector() {constant=<rgba 1 0.0393 0.1892 0>}
  n51 = LinearInterpolate(A=Constant3Vector() {constant=<rgba 0 0.6353 1 0>}, B=Constant3Vector() {constant=<rgba 0 1 0.8509 0>}, Alpha=Multiply(A=@n4, B=@n40))
  n52 = Add(A=Multiply(A=LinearInterpolate(A=@n49, B=@n50, Alpha=@n21), B=@n21), B=Multiply(A=Multiply(A=@n51) {constB=2}, B=@n40))
  n53 = ScalarParameter() {defaultValue=1, parameterName=WaveBrightness}
  MaterialFunctionCall(Emissive=Multiply(A=Multiply(A=@n52, B=@n53), B=@n1.RGB)) {materialFunction=&MF_FX_EmissiveAdjustments}
  n54 = Add(A=SphereMask(A=WorldPosition().XYZ, B=ActorPositionWS()) {attenuationRadius=10000, hardnessPercent=0}, B=Divide(A=Time()) {constB=10})
  n55 = Multiply(
   A=Transform(None=TextureSample(UVs=@n54) {samplerType=SAMPLERTYPE_LinearColor, texture=&T_NoisePack_LinearColor}.RGB)
   B=ScalarParameter() {parameterName=WPO_Wave}
  )
  n56 = Multiply(A=ScalarParameter() {parameterName=WPO_Wave_Detail}, B=Transform(None=@n24.RGB))
  n57 = Multiply(A=DynamicParameter() {paramNames [WaveDistortion, WPO, Tile_Wave, TimeOffset]}.WPO, B=Multiply(A=VertexColor().R, B=Add(A=@n55, B=@n56)))
  n58 = Constant()
  ShadingPathSwitch(Default=@n57, Deferred=@n57, Forward=@n58, Mobile=@n58)
 ]
}
