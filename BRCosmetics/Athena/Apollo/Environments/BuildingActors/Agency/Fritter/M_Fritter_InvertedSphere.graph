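# Translucent unlit effect material for mesh particles, named for an inverted sphere in the Agency Fritter environment set.
# Emissive blends two colors through a noise-distorted panning mask, weighted by view angle, plus a dim teal scrolling layer.
# A view-facing term times particle alpha feeds temporal dither, fading silhouette edges.
# A separate branch builds a flickering, noise-varied offset along vertex normals, confined to bands by a cosine of the vertical UV gradient.
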Material /BRCosmetics/Athena/Apollo/Environments/BuildingActors/Agency/Fritter/M_Fritter_InvertedSphere
graph {
 properties {
  bUsedWithMeshParticles=true
  bUsedWithNiagaraMeshParticles=true
  blendMode=BLEND_Translucent
  pixelDepthOffsetMode=PDOM_Legacy
  shadingModel=MSM_Unlit
  usedShadingModels=Unlit
 }
 nodes [
  n1 = Abs(None=DotProduct(A=CameraVectorWS(), B=VertexNormalWS()))
  n2 = Multiply(A=@n1) {constB=1.7}
  n3 = ParticleColor()
  MaterialFunctionCall(Alpha Threshold=Multiply(A=Saturate(None=@n2), B=@n3.A)) {materialFunction=&DitherTemporalAA}
  n4 = VectorParameter() {parameterName=UV_Tile, defaultValue=<rgba 1 2 0 0>}
  n5 = Multiply(A=TextureCoordinate(), B=AppendVector(A=@n4.R, B=@n4.R))
  n6 = Panner() {speedX=-0.02, speedY=0.2}
  n7 = Divide(A=ComponentMask(None=TextureSample(UVs=@n6) {samplerType=SAMPLERTYPE_LinearColor, texture=&T_NoisePack_LinearColor}.RGB) {g=true, r=true})
  n8 = TextureSample(UVs=Panner(Coordinate=Add(A=@n5, B=@n7)) {speedX=0.2, speedY=-0.25}) {texture=/Game/Weapons/WeaponSkins/Wraps/Textures/Mash/Mash_M}
  n9 = Multiply(A=@n8.R, B=@n8.R)
  n10 = CameraVectorWS()
  n11 = VertexNormalWS()
  n12 = Power(Base=Abs(None=DotProduct(A=@n10, B=@n11)), Exp=ScalarParameter() {defaultValue=4, parameterName=BrightNoiseEdgeFadeExp}) {constExponent=4}
  n13 = Divide(A=LinearInterpolate(
    A=VectorParameter() {parameterName=ColorLow, defaultValue=<rgba 5 0.925 0.165 1>}.RGB
    B=VectorParameter() {parameterName=ColorHigh, defaultValue=<rgba 1 0 0 1>}.RGB
    Alpha=Multiply(A=@n9, B=@n12)
   )) {constB=10}
  n14 = Constant3Vector() {constant=<rgba 0.0063 0.0531 0.07 0>}
  n15 = TextureSample(UVs=Panner(Coordinate=TextureCoordinate() {uTiling=2, vTiling=2}) {speedY=-0.3}) {texture=&Mash_M}
  MaterialFunctionCall(Emissive=Add(A=Multiply(A=@n13, B=@n8.R), B=Multiply(A=@n14, B=@n15.R))) {materialFunction=&MF_FX_EmissiveAdjustments}
  Multiply(A=Multiply(A=@n1) {constB=2}, B=@n3.A)
  n16 = Multiply(A=MaterialFunctionCall() {materialFunction=&LinearGradient}.VGradient)
  n17 = Subtract(A=OneMinus(None=Cosine(None=@n16))) {constB=0.6}
  n18 = LinearInterpolate(Alpha=TextureSample(UVs=TextureCoordinate() {uTiling=5, vTiling=0}) {samplerType=SAMPLERTYPE_LinearColor, texture=&T_NoisePack_LinearColor}.R) {constA=-1}
  n19 = Multiply(A=VertexNormalWS()) {constB=-1}
  n20 = MaterialFunctionCall!(FlickerInitialSpeed, FlickerRandom) {materialFunction=/Game/Effects/Fort_Effects/Materials/_Functions/MF_Flicker}
  Multiply(
   A=Power(Base=Saturate(None=@n17)) {constExponent=10}
   B=Multiply(A=Multiply(A=@n18, B=@n19), B=LinearInterpolate(Alpha=@n20.Result) {constB=15}) {constB=15}
  )
 ]
}
