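# Light function material for the Agency Fritter building actors that builds an animated, wave-like intensity pattern for lights.
# A Mash wrap texture and a linear-color noise pack share V-tiled UVs, panned diagonally at different speeds.
# The wrap's red channel times the squared noise green channel drives a lerp that boosts intensity above one at peaks.
# An Add of both red channels feeds nothing.
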Material /BRCosmetics/Athena/Apollo/Environments/BuildingActors/Agency/Fritter/M_Fritter_Light_Wave
graph {
 properties {materialDomain=MD_LightFunction, pixelDepthOffsetMode=PDOM_Legacy, usedShadingModels=Default Lit}
 nodes [
  n1 = TextureCoordinate() {vTiling=3}
  n2 = TextureSample(UVs=Panner(Coordinate=@n1) {speedX=0.15, speedY=0.3}) {texture=/Game/Weapons/WeaponSkins/Wraps/Textures/Mash/Mash_M}
  n3 = TextureSample(UVs=Panner(Coordinate=@n1) {speedX=0.05, speedY=0.1}) {samplerType=SAMPLERTYPE_LinearColor, texture=&T_NoisePack_LinearColor}
  Add(A=@n2.R, B=@n3.R)
  LinearInterpolate(Alpha=Multiply(A=@n2.R, B=Multiply(A=@n3.G, B=@n3.G))) {constB=5}
 ]
}
